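; Gameplay ability for pawns in wind spline fields that applies a block-riding gameplay effect while the Fort pawn is skydiving.
; On activation, authority binds to the pawn's movement-mode change delegate and runs the handler once for the current state.
; The handler checks for the Skydiving custom movement mode, using a flag and stored handle to apply or remove the effect once per transition.
; Ending the ability unbinds the delegate and clears the effect on authority.
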(event Ability|EventActivateAbility
  (Ability|HasAuthority
    (:True
      (bind _asfort_pawn (Utilities|Casting|CastToFortPawn (Ability|GetAvatarActorfromActorInfo)))
      (bind _output_get (Variables|Default|SetFort%20Pawn _asfort_pawn))
      (Character|BindEventtoMovementModeChangedDelegate _output_get (EventDispatchers|CreateEvent :!event Custom|HandleMovementModeChanged))
      (bind _asfort_movement_comp_character (Utilities|Casting|CastToFortMovementComp_Character (Pawn|GetMovementComponent (Variables|Default|GetFort%20Pawn))))
      (CallFunction|HandleMovementModeChanged (Variables|Default|GetFort%20Pawn) (Variables|/Script/Engine.CharacterMovementComponent|GetMovementMode _asfort_movement_comp_character) (Variables|/Script/Engine.CharacterMovementComponent|GetCustomMovementMode _asfort_movement_comp_character)))))

(event Ability|EventOnEndAbility
  (params (bWasCancelled :type Boolean))
  (bind _fort_pawn (Variables|Default|GetFort%20Pawn))
  (Ability|HasAuthority
    (:True
      (Utilities|IsValid _fort_pawn
        (:Is%20Valid
          (Character|UnbindEventfromMovementModeChangedDelegate _fort_pawn (EventDispatchers|CreateEvent :!event Custom|HandleMovementModeChanged))
          (CallFunction|HandleNotSkydiving))))))

(event Custom|HandleMovementModeChanged
  (params (Character :type Character%20Object%20Reference) (PrevMovementMode :type EMovementMode%20Enum :default "MOVE_None") (PreviousCustomMode :type Byte))
  (bind _asfort_movement_comp_character (Utilities|Casting|CastToFortMovementComp_Character (Pawn|GetMovementComponent Character)))
  (if (== (Variables|/Script/Engine.CharacterMovementComponent|GetCustomMovementMode _asfort_movement_comp_character) (Utilities|Enum|LiteralenumEFortCustomMovement "Skydiving"))
    (CallFunction|HandleSkydiving)
    (else
      (CallFunction|HandleNotSkydiving))))

(event Custom|HandleSkydiving
  (if (not (Variables|Default|GetIsSkydiving))
    (Variables|Default|SetIsSkydiving true)
    (bind _returnvalue (Ability|ApplyGameplayEffectToOwner "/Wind/Gameplay/WindFields/Parent/GE_WindSpline_Pawn_BlockRiding.GE_WindSpline_Pawn_BlockRiding_C"))
    (Variables|Default|SetBlockRidingHandle _returnvalue)))

(event Custom|HandleNotSkydiving
  (if (Variables|Default|GetIsSkydiving)
    (Variables|Default|SetIsSkydiving false)
    (Ability|RemoveGameplayEffectFromOwnerWithHandle (Variables|Default|GetBlockRidingHandle))))

(event Ability|EventFailedToActivatePassiveAbility
  (params (ActorInfo :type Gameplay%20Ability%20Actor%20Info%20Structure)))
