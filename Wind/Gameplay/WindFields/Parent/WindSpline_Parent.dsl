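; Parent Blueprint for wind spline wind fields, pairing a wind spline with optional wind and snow mist effects plus spline audio.
; The construction script destroys unused effect components, pushes wind color and gust spawn scale to the wind effect, caches spline length and offsets the snow mist.
; Wind spline enabled and disabled events switch the snow mist effect and the spline audio on and off.
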(event OnWindSplineDisabled(WindSpline)
  (bind _ns_snowmist_kissworld (Variables|Default|GetNS_SnowMist_KissWorld)
    (:then
      (Components|Activation|Deactivate _ns_snowmist_kissworld))
    (:else
      _))
  (Shape|Disable (Variables|Default|GetFortSplineAudio)))

(event OnWindSplineEnabled(WindSpline)
  (bind _ns_snowmist_kissworld (Variables|Default|GetNS_SnowMist_KissWorld)
    (:then
      (Components|Activation|Activate _ns_snowmist_kissworld false))
    (:else
      _))
  (Shape|Enable (Variables|Default|GetFortSplineAudio)))

(fn ConstructionScript
  (Variables|Default|GetbUseWindSplineFx
    (:then
      _)
    (:else
      (Components|DestroyComponent (Variables|Default|GetNS_WindSpline_InWorld))))
  (Variables|Default|GetbUseSnowMistFx
    (:then
      _)
    (:else
      (Components|DestroyComponent (Variables|Default|GetNS_SnowMist_KissWorld))))
  (Utilities|FlowControl|Sequence
    (:then_0
      (bind _ns_windspline_inworld (Variables|Default|GetNS_WindSpline_InWorld)
        (:then
          (Effects|Components|ParticleSystem|SetColorParameter _ns_windspline_inworld "WindColor" (Variables|Default|GetWindColor))
          (Effects|Components|ParticleSystem|SetFloatParameter _ns_windspline_inworld "WindGustSpawnScale" (Variables|Default|GetWindGustSpawnScale)))
        (:else
          _))
      (Variables|Default|SetSplineLength (Spline|GetSplineLength (Variables|Default|GetWindSpline))))
    (:then_1
      (bind _ns_snowmist_kissworld (Variables|Default|GetNS_SnowMist_KissWorld)
        (:then
          (Transformation|AddLocalOffset _ns_snowmist_kissworld (Variables|Default|GetSnowMistLocation) false false)
          (Transformation|AddLocalRotation _ns_snowmist_kissworld (Variables|Default|GetSnowMistRotation) false false))
        (:else)))))
